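# Physics asset for the F_MED_IceRetreat face-accessory mesh, covering secondary motion of the player cosmetic's hair strands and earrings.
# Small base capsules attach to the head with free angular constraints, and each anchors a dyn_ capsule with limited swing and tight twist.
# Some front strands chain a second dyn_ segment.
# Head, neck, clavicle and upper-arm capsules supply body shapes, though the left upper arm's is tiny compared to the right's.
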PhysicsAsset /BRCosmetics/Characters/Player/Female/Medium/Bodies/F_MED_IceRetreat/Meshes/Parts/F_MED_IceRetreat_FaceAcc_Physics
bodies {
 base_earring_l [{shapeType=Capsule, length=0.5, radius=0.5, shapeName=base_earring_l_capsule}]
 base_earring_r [{shapeType=Capsule, length=0.5, radius=0.5, shapeName=base_earring_r_capsule}]
 base_hair_l_back [{shapeType=Capsule, length=0.5, radius=0.5, shapeName=base_hair_l_back_capsule}]
 base_hair_l_front_01 [{shapeType=Capsule, length=0.5, radius=0.5, shapeName=base_hair_l_front_01_capsule}]
 base_hair_l_front_02 [{shapeType=Capsule, length=0.5, radius=0.5, shapeName=base_hair_l_front_02_capsule}]
 base_hair_l_front_03 [{shapeType=Capsule, length=0.5, radius=0.5, shapeName=base_hair_l_front_03_capsule}]
 base_hair_l_middl [{shapeType=Capsule, length=0.5, radius=0.5, shapeName=base_hair_l_middl_capsule}]
 base_hair_middl_front [{shapeType=Capsule, length=0.5, radius=0.5, shapeName=base_hair_middl_front_capsule}]
 base_hair_r_back [{shapeType=Capsule, length=0.5, radius=0.5, shapeName=base_hair_r_back_capsule}]
 base_hair_r_front_01 [{shapeType=Capsule, length=0.5, radius=0.5, shapeName=base_hair_r_front_01_capsule}]
 base_hair_r_front_03 [{shapeType=Capsule, length=0.5, radius=0.5, shapeName=base_hair_r_front_03_capsule}]
 base_hair_r_middl [{shapeType=Capsule, length=0.5, radius=0.5, shapeName=base_hair_r_middl_capsule}]
 clavicle_l [{shapeType=Capsule, length=9, radius=6.6, shapeName=clavicle_l_capsule, center=<5.4727 0.7537 -2.1083>, rotation=<rot 79.007 -0.0002 -0.0002>}]
 clavicle_r [{shapeType=Capsule, length=9, radius=6.6, shapeName=clavicle_l_capsule, center=<-5.4727 0.7537 2.1083>, rotation=<rot 79.007 -0.0002 -0.0002>}]
 dyn_earring_l [{shapeType=Capsule, length=4, radius=0.8, shapeName=dyn_earring_l_capsule, center=<2.5 0.7 0>, rotation=<rot 90 0 0>}]
 dyn_earring_r [{shapeType=Capsule, length=0.5, radius=2.8, shapeName=dyn_earring_l_capsule, center=<2.5 1.6 0>, rotation=<rot 0 0 90>}]
 dyn_hair_l_back_01 [{shapeType=Capsule, length=5, radius=2, shapeName=dyn_hair_l_back_01_capsule, center=<2.6 0 0.8697>, rotation=<rot 90 0 0>}]
 dyn_hair_l_front_01 [{shapeType=Capsule, length=4, radius=1, shapeName=dyn_hair_l_front_01_capsule, center=<2 0 0>, rotation=<rot 90 0 0>}]
 dyn_hair_l_front_02 [{shapeType=Capsule, length=3, radius=1, shapeName=dyn_hair_l_front_02_capsule, center=<1.5 0 0>, rotation=<rot 90 0 0>}]
 dyn_hair_l_front_03 [{shapeType=Capsule, length=1.5, radius=1, shapeName=dyn_hair_l_front_03_capsule, center=<0.75 0 0>, rotation=<rot 90 0 0>}]
 dyn_hair_l_front_04 [{shapeType=Capsule, length=3, radius=1, shapeName=dyn_hair_l_front_04_capsule, center=<1.5 0 -0.1938>, rotation=<rot 90 0 0>}]
 dyn_hair_l_middl_01 [{shapeType=Capsule, length=5, radius=2, shapeName=dyn_hair_l_middl_01_capsule, center=<2.6 0 0.1697>, rotation=<rot 90 0 0>}]
 dyn_hair_middl_front [{shapeType=Capsule, length=2.5, radius=1, shapeName=dyn_hair_middl_front_capsule, center=<1.25 0 0>, rotation=<rot 90 -113.6055 -113.6055>}]
 dyn_hair_r_back_01 [{shapeType=Capsule, length=5, radius=2, shapeName=dyn_hair_r_back_01_capsule, center=<4.5663 -1 0.4>, rotation=<rot 90 179.5524 -175.4478>}]
 dyn_hair_r_front_01 [{shapeType=Capsule, length=1, radius=1, shapeName=dyn_hair_r_front_01_capsule, center=<0.5 0 0>, rotation=<rot 90 0 0>}]
 dyn_hair_r_front_02 [{shapeType=Capsule, length=3, radius=1, shapeName=dyn_hair_r_front_02_capsule, center=<1.5 0 0>, rotation=<rot 90 0 0>}]
 dyn_hair_r_front_03 [{shapeType=Capsule, length=2, radius=1, shapeName=dyn_hair_r_front_03_capsule, center=<1 0 0>, rotation=<rot 90 0 0>}]
 dyn_hair_r_front_04 [{shapeType=Capsule, length=5, radius=1, shapeName=dyn_hair_r_front_04_capsule, center=<2.2547 0.4 -0.2516>, rotation=<rot 60 0 -10>}]
 dyn_hair_r_middl_01 [{shapeType=Capsule, length=5, radius=1.5, shapeName=dyn_hair_r_middl_01_capsule, center=<3.5 0 -1>, rotation=<rot 90 -179.9975 -179.9975>}]
 head [{shapeType=Capsule, length=6, radius=6.5, shapeName=head_capsule, center=<0.9 1.7 0>, rotation=<rot 90 180 150>}]
 neck_01 [{shapeType=Capsule, length=4, radius=4.8, shapeName=neck_01_capsule, center=<0 0.6175 0>, rotation=<rot 90 0 0>}]
 upperarm_l [{shapeType=Capsule, length=0.505, radius=0.505, shapeName=upperarm_l_capsule, rotation=<rot 0 -90 0>}]
 upperarm_r [{shapeType=Capsule, length=11, radius=5.5, shapeName=clavicle_l_capsule, center=<-8 0 0.1>, rotation=<rot 90 0 0>}]
}
constraints [
 {
  bone1Name=base_earring_l
  bone2Name=head
  swing1LimitDegrees=45
  swing1Motion=Free
  swing2LimitDegrees=45
  swing2Motion=Free
  twistLimitDegrees=45
  twistMotion=Free
 }
 {bone1Name=dyn_earring_l, bone2Name=base_earring_l, swing1LimitDegrees=15, swing2LimitDegrees=15, twistLimitDegrees=5}
 {
  bone1Name=base_earring_r
  bone2Name=head
  swing1LimitDegrees=45
  swing1Motion=Free
  swing2LimitDegrees=45
  swing2Motion=Free
  twistLimitDegrees=45
  twistMotion=Free
 }
 {bone1Name=dyn_earring_r, bone2Name=base_earring_r, swing1LimitDegrees=15, swing2LimitDegrees=25, twistLimitDegrees=5}
 {
  bone1Name=base_hair_l_back
  bone2Name=head
  swing1LimitDegrees=45
  swing1Motion=Free
  swing2LimitDegrees=45
  swing2Motion=Free
  twistLimitDegrees=45
  twistMotion=Free
 }
 {bone1Name=dyn_hair_l_back_01, bone2Name=base_hair_l_back, swing1LimitDegrees=40, swing2LimitDegrees=20, twistLimitDegrees=5}
 {
  bone1Name=base_hair_l_front_01
  bone2Name=head
  swing1LimitDegrees=45
  swing1Motion=Free
  swing2LimitDegrees=45
  swing2Motion=Free
  twistLimitDegrees=45
  twistMotion=Free
 }
 {bone1Name=dyn_hair_l_front_01, bone2Name=base_hair_l_front_01, swing1LimitDegrees=10, swing2LimitDegrees=10, twistLimitDegrees=5}
 {
  bone1Name=base_hair_l_front_02
  bone2Name=head
  swing1LimitDegrees=45
  swing1Motion=Free
  swing2LimitDegrees=45
  swing2Motion=Free
  twistLimitDegrees=45
  twistMotion=Free
 }
 {bone1Name=dyn_hair_l_front_02, bone2Name=base_hair_l_front_02, swing1LimitDegrees=10, swing2LimitDegrees=10, twistLimitDegrees=5}
 {
  bone1Name=base_hair_l_front_03
  bone2Name=head
  swing1LimitDegrees=45
  swing1Motion=Free
  swing2LimitDegrees=45
  swing2Motion=Free
  twistLimitDegrees=45
  twistMotion=Free
 }
 {bone1Name=dyn_hair_l_front_03, bone2Name=base_hair_l_front_03, swing1LimitDegrees=13, swing2LimitDegrees=10, twistLimitDegrees=5}
 {bone1Name=dyn_hair_l_front_04, bone2Name=dyn_hair_l_front_03, swing1LimitDegrees=15, swing2LimitDegrees=15, twistLimitDegrees=5}
 {
  bone1Name=base_hair_l_middl
  bone2Name=head
  swing1LimitDegrees=45
  swing1Motion=Free
  swing2LimitDegrees=45
  swing2Motion=Free
  twistLimitDegrees=45
  twistMotion=Free
 }
 {bone1Name=dyn_hair_l_middl_01, bone2Name=base_hair_l_middl, swing1LimitDegrees=35, swing2LimitDegrees=30, twistLimitDegrees=5}
 {
  bone1Name=base_hair_middl_front
  bone2Name=head
  swing1LimitDegrees=45
  swing1Motion=Free
  swing2LimitDegrees=45
  swing2Motion=Free
  twistLimitDegrees=45
  twistMotion=Free
 }
 {bone1Name=dyn_hair_middl_front, bone2Name=base_hair_middl_front, swing1LimitDegrees=10, swing2LimitDegrees=10, twistLimitDegrees=5}
 {
  bone1Name=base_hair_r_back
  bone2Name=head
  swing1LimitDegrees=45
  swing1Motion=Free
  swing2LimitDegrees=45
  swing2Motion=Free
  twistLimitDegrees=45
  twistMotion=Free
 }
 {bone1Name=dyn_hair_r_back_01, bone2Name=base_hair_r_back, swing1LimitDegrees=35, swing2LimitDegrees=20, twistLimitDegrees=5}
 {
  bone1Name=base_hair_r_front_01
  bone2Name=head
  swing1LimitDegrees=45
  swing1Motion=Free
  swing2LimitDegrees=45
  swing2Motion=Free
  twistLimitDegrees=45
  twistMotion=Free
 }
 {bone1Name=dyn_hair_r_front_01, bone2Name=base_hair_r_front_01, swing1LimitDegrees=10, swing2LimitDegrees=10, twistLimitDegrees=5}
 {bone1Name=dyn_hair_r_front_02, bone2Name=dyn_hair_r_front_01, swing1LimitDegrees=10, swing2LimitDegrees=10, twistLimitDegrees=5}
 {
  bone1Name=base_hair_r_front_03
  bone2Name=head
  swing1LimitDegrees=45
  swing1Motion=Free
  swing2LimitDegrees=45
  swing2Motion=Free
  twistLimitDegrees=45
  twistMotion=Free
 }
 {bone1Name=dyn_hair_r_front_03, bone2Name=base_hair_r_front_03, swing1LimitDegrees=13, swing2LimitDegrees=10, twistLimitDegrees=5}
 {bone1Name=dyn_hair_r_front_04, bone2Name=dyn_hair_r_front_03, swing1LimitDegrees=15, swing2LimitDegrees=15, twistLimitDegrees=5}
 {
  bone1Name=base_hair_r_middl
  bone2Name=head
  swing1LimitDegrees=45
  swing1Motion=Free
  swing2LimitDegrees=45
  swing2Motion=Free
  twistLimitDegrees=45
  twistMotion=Free
 }
 {bone1Name=dyn_hair_r_middl_01, bone2Name=base_hair_r_middl, swing1LimitDegrees=35, swing2LimitDegrees=20, twistLimitDegrees=5}
 {
  bone1Name=head
  bone2Name=neck_01
  swing1LimitDegrees=45
  swing1Motion=Free
  swing2LimitDegrees=45
  swing2Motion=Free
  twistLimitDegrees=45
  twistMotion=Free
 }
 {
  bone1Name=upperarm_r
  bone2Name=clavicle_r
  swing1LimitDegrees=45
  swing1Motion=Free
  swing2LimitDegrees=45
  swing2Motion=Free
  twistLimitDegrees=45
  twistMotion=Free
 }
 {
  bone1Name=upperarm_l
  bone2Name=clavicle_l
  swing1LimitDegrees=45
  swing1Motion=Free
  swing2LimitDegrees=45
  swing2Motion=Free
  twistLimitDegrees=45
  twistMotion=Free
 }
]
